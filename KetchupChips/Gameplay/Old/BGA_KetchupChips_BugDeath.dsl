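; BGA_KetchupChips_BugDeath is a short-lived actor Blueprint, apparently the goo left by a bug death in KetchupChips, stored in an Old folder.
; On BeginPlay it calls its parent, then sets a brief timer by event bound to LifetimeGooDespawn, which destroys the actor.
; The overlap event and construction script only call their parents, and Tick is an unused, disabled stub.
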(event EventTick
  (params (DeltaSeconds :type Float%20%28single-precision%29))
  (|CallParentFunction DeltaSeconds) ; This node is disabled and will not be called.\nDrag off pins to build functionality.
)

(event EventBeginPlay
  (|CallParentFunction)
  (Utilities|Time|SetTimerbyEvent (delegate Custom|LifetimeGooDespawn) 0.1))

(event Custom|LifetimeGooDespawn
  (Actor|DestroyActor self))

(event Collision|EventActorBeginOverlap
  (params (OtherActor :type Actor%20Object%20Reference))
  (|CallParentFunction OtherActor))

(fn ConstructionScript
  (|CallParentFunction))
